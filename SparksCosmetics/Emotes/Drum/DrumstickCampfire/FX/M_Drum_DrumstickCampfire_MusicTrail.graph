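# Unlit translucent Niagara material for the music-staff trail in the Drumstick Campfire drum emote.
# A panning distortion normal map offsets the UVs of the music staff texture, which scrolls horizontally.
# The staff's red channel, through a dynamic-parameter desaturation, intensity and particle color, feeds an emissive adjustment function.
# Opacity math multiplies that scrolling red channel by a static green-channel mask, then applies depth fade and particle alpha.
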Material /SparksCosmetics/Emotes/Drum/DrumstickCampfire/FX/M_Drum_DrumstickCampfire_MusicTrail
graph {
 properties {
  bUsedWithNiagaraMeshParticles=true
  bUsedWithNiagaraRibbons=true
  bUsedWithNiagaraSprites=true
  blendMode=BLEND_Translucent
  pixelDepthOffsetMode=PDOM_Legacy
  shadingModel=MSM_Unlit
  usedShadingModels=Unlit
 }
 nodes [
  n1 = NamedRerouteUsage()
  n2 = TextureCoordinate()
  n3 = TextureSample(UVs=Panner(Coordinate=Add(A=@n1, B=@n2)) {speedX=-0.6}) {texture=&/TX_Drum_Drumstick_MusicStaff}
  n4 = Desaturation(None=@n3.R, Fraction=DynamicParameter() {defaultValue=<rgba 0 1 1 1>, paramNames [Desaturation, Panning Y, Param3, Param4]}.Desaturation)
  n5 = ScalarParameter() {defaultValue=0.5, group=Emissive, parameterName=Emissive Intensity}
  n6 = ParticleColor()
  MaterialFunctionCall(Emissive=Multiply(A=Multiply(A=@n4, B=@n5), B=@n6.RGB)) {materialFunction=&MF_FX_EmissiveAdjustments}
  n7 = TextureCoordinate()
  n8 = VectorParameter() {group=Distortion, parameterName=Distortion UV, sortPriority=1, defaultValue=<rgba 0.6 0.6 -1 0>}
  n9 = AppendVector(A=@n8.R, B=@n8.R)
  n10 = TextureSampleParameter2D(UVs=Panner(Coordinate=Multiply(A=@n7, B=@n9), Speed=AppendVector(A=@n8.B, B=@n8.B))) {
   group=Distortion
   parameterName="Distortion  Tex"
   samplerType=SAMPLERTYPE_Normal
   sortPriority=0
   texture=&T_Weap_Distort_N_02
  }
  n11 = Multiply(
   A=ComponentMask(None=@n10.RGB) {g=true, r=true}
   B=ScalarParameter() {defaultValue=0.08, group=Distortion, parameterName=Distortion Intensity, sortPriority=2}
  )
  Multiply(
   A=LinearInterpolate(
    A=ScalarParameter() {defaultValue=1, group=Emissive, parameterName=Emissive Dark}
    B=ScalarParameter() {defaultValue=1, group=Emissive, parameterName=Emissive Light}
    Alpha=@n11
   ) {constA=0.1, constB=1.5}
   B=@n4
  )
  n12 = TextureCoordinate()
  n13 = Multiply(
   A=@n3.R
   B=Power(Base=TextureSample(UVs=@n12) {texture=&TX_Drum_Drumstick_MusicStaff}.G) {constExponent=1.25}
  )
  n14 = ScalarParameter() {defaultValue=1, group=Opactiy, parameterName=Opactiy Intensity}
  Multiply(A=@n6.A, B=DepthFade(Opacity=Multiply(A=@n13, B=@n14)) {fadeDistanceDefault=75})
  NamedRerouteDeclaration(None=@n11) {name=Distortion, nodeColor=<rgba 0.4062 1 0 1>}
 ]
}
