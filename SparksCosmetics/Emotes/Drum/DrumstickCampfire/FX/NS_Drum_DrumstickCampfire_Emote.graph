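# Niagara effect for the Drumstick Campfire drum emote: a multicolored campfire with embers, scorch glow and music trail.
# GPU sprite flames layer with three color-oscillating flame meshes, while curl-noise embers and a sine-waved, rainbow-tinted ribbon rise above.
# Scorch combines sprite, mesh and a jittering light that is zeroed in the front end.
# Spawn-group masks vary flame layers, time-of-day scalars adjust brightness, and Center_flame is disabled.
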NiagaraSystem /SparksCosmetics/Emotes/Drum/DrumstickCampfire/FX/NS_Drum_DrumstickCampfire_Emote
dependencies {
 usedDataInterfaces [
  &NiagaraDataInterfaceArrayColor
  &NiagaraDataInterfaceCurve
  &NiagaraDataInterfaceSpriteRendererInfo
  &NiagaraDataInterfaceColorCurve
  &NiagaraDataInterfaceVectorCurve
  &NiagaraDataInterfaceVector2DCurve
 ]
 usedDynamicInputs []
 usedModules [
  &self:SystemSpawnScript.SystemScriptSource.SystemScriptGraph.NiagaraNodeAssignment_0.SetVariables_#1
  &SystemState
 ]
}
emitters {
 Center_flame {
  bEnabled=false
  emitterUpdateScript [
   &EmitterState(Life Cycle Mode=::NewEnumerator1, Loop Behavior=::NewEnumerator1, Loop Delay=1.4, Loop Duration Mode=::NewEnumerator1, UseLoopDelay=false)
   &SpawnBurst_Instantaneous(Spawn Count=1, Use Spawn Probability=false)
   &SpawnBurst_Instantaneous(Spawn Group=1, Use Spawn Probability=false, UseLoopCountLimit=false) {enabled=false, module=SpawnBurst_Instantaneous001}
   &SpawnBurst_Instantaneous(Spawn Group=2, Use Spawn Probability=false, UseLoopCountLimit=false) {enabled=false, module=SpawnBurst_Instantaneous002}
  ]
  particleSpawnScript [
   &InitializeParticle(
    Mesh Scale=&MaskVectorBySpawnGroup(
     Masked Vector=<1.05 0.8 1.05>
     Passthrough Vector=&MaskVectorBySpawnGroup(Masked Vector=<0.8 0.65 0.8>, Passthrough Vector=<1 0.8 1>, Spawn Group Mask=1)
    )
    Mesh Scale Mode=::NewEnumerator4
    UsePositionOffset=false
    Write Material Random=true
   ) {module=InitializeParticle001}
   &InitialMeshOrientation(
    Mesh Axis To Orient=<0 1 0>
    Mesh Orientation Mode=::NewEnumerator2
    Orientation Vector=<0 0 1>
    Rotation=&MaskVectorBySpawnGroup(
     Masked Vector=<0 0.25 0>
     Passthrough Vector=&MaskVectorBySpawnGroup(Masked Vector=<0 -0.15 0>, Passthrough Vector=<0 0.05 0>, Spawn Group Mask=1)
    )
    Use Rotation Vector=true
   )
   &CameraOffset(Camera Offset Amount=&MaskFloatBySpawnGroup(Masked Float=0, Passthrough Float=&MaskFloatBySpawnGroup(Masked Float=12, Passthrough Float=0)))
  ]
  particleUpdateScript [
   &ParticleState(
    Kill Particles When Lifetime Has Elapsed=false
    Let Infinitely Lived Particles Die When Emitter Deactivates=true
    Loop Particles Lifetime=true
   )
   &ScaleMeshSize(Scale Factor=&VectorFromFloat(Value=&FloatFromCurve(FloatCurve=DataInterface() {propertyValues {Curve {keys [<at 0 0>,<at 0.25 1>]}}}, CurveIndex=$Emitter.Age)))
   &SolveForcesAndVelocity(Clamp Velocity=false, Limit Acceleration=false)
   /SetVariables(Particles.MainColor:LinearColor=&MultiplyLinearColor(
     Linear Color=&MultiplyLinearColor(
      Linear Color=&MaskLinearColorBySpawnGroup(
       Masked Linear Color=<rgba 2.5 0.2805 0.001 1>
       Passthrough Linear Color=&MaskLinearColorBySpawnGroup(Masked Linear Color=<rgba 2.1167 0 3 1>, Passthrough Linear Color=<rgba 0.5805 1 0 1>, Spawn Group Mask=1)
      )
      Scale Factor=&MakeVector4_FromLinearColor(LinearColor=&ColorFromCurve(
        ColorCurve=DataInterface() {propertyValues {
          AlphaCurve {keys [<at 0 0 RCIM_Cubic>,<at 0.1 1>]}
          BlueCurve {keys [<at 0 1 RCIM_Cubic>,<at 1 1 RCIM_Cubic>]}
          GreenCurve {keys [<at 0 1 RCIM_Cubic>,<at 1 1 RCIM_Cubic>]}
          RedCurve {keys [<at 0 1 RCIM_Cubic>,<at 1 1 RCIM_Cubic>]}
         }}
        CurveIndex=$Emitter.Age
       ))
     )
     Scale Factor=&MakeVector4_FromFloat(Float=&FNTimeOfDayScalar(Night Scale=0.75, Morning Scale=1, Day Scale=0.7, Evening Scale=1))
    )) {isSetParameters=true}
   &DynamicMaterialParameters(
    Index 0 Param 1=&MaskFloatBySpawnGroup(
     Masked Float=1
     Passthrough Float=&MaskFloatBySpawnGroup(Masked Float=1, Passthrough Float=1.5, Spawn Group Mask=1)
    )
    Module.Index0Float0Write=true
    Module.Index0Float1Write=true
    Module.Index0Float2Write=true
    Module.Index0Float3Write=true
   )
   /SetVariables(Particles.Position=&MaskVectorBySpawnGroup(
     Masked Vector=&AddVectorToPosition(Position=$Engine.Emitter.SimulationPosition, Vector=<0 0 -12>)
     Passthrough Vector=&MaskVectorBySpawnGroup(
      Masked Vector=&AddVectorToPosition(
       Position=$Engine.Emitter.SimulationPosition
       Vector=&MakeVector(X=&Waveform("Waveform [1]"=::NewEnumerator2, "[1] Frequency"=0.4, "[1] Amplitude Min/Max"=<-4 4>), Y=0, Z=-11)
      )
      Passthrough Vector=&AddVectorToPosition(
       Position=$Engine.Emitter.SimulationPosition
       Vector=&MakeVector(X=&Waveform("Waveform [1]"=::NewEnumerator0, "[1] Frequency"=0.1, "[1] Amplitude Min/Max"=<-12 0>), Y=0, Z=-15)
      )
      Spawn Group Mask=1
     )
    )) {isSetParameters=true}
  ]
  rendererClasses [&NiagaraMeshRendererProperties]
  renderers [{rendererClass=&NiagaraMeshRendererProperties, rendererIndex=0}]
 }
 Embers {
  emitterUpdateScript [
   &EmitterState(
    Enable Distance Culling=true
    Enable Visibility Culling=true
    Life Cycle Mode=::NewEnumerator1
    Loop Behavior=::NewEnumerator1
    Loop Duration Mode=::NewEnumerator1
    MaxDistance=2000
    Module.bUseMaxDistance=true
    Module.bUseMinDistance=false
    Scalability Mode=::NewEnumerator1
    UseLoopDelay=false
   )
   &SpawnBurst_Instantaneous(Spawn Count=8, Spawn Time=0.08, Use Spawn Probability=false, UseLoopCountLimit=false)
   &SpawnBurst_Instantaneous(Spawn Count=24, Spawn Group=1, Spawn Time=0.08, Use Spawn Probability=false, UseLoopCountLimit=false) {module=SpawnBurst_Instantaneous001}
   &SpawnRate(SpawnRate=3, Use Spawn Probability=false)
   &SpawnRate(SpawnGroup=1, SpawnRate=8, Use Spawn Probability=false) {module=SpawnRate001}
  ]
  particleSpawnScript [
   /SetVariables(Particles.VisibilityTag=&SpawnGroup()) {isSetParameters=true}
   &InitializeParticle(
    Color=&MaskLinearColorBySpawnGroup(
     Masked Linear Color=&SelectLinearColorFromArray(
      Array Sampling Mode=::NewEnumerator0
      Direct Array Index=?
      Color Selection Array=DataInterface() {propertyValues {ColorData [
         <rgba 1 0.0981 0 1>
         <rgba 1 0.0981 0 1>
         <rgba 1 0.0981 0 1>
         <rgba 1 0.0981 0 1>
         <rgba 0 0.1984 1 1>
         <rgba 0 0.1984 1 1>
         <rgba 0.102 0.0339 1 1>
         <rgba 0 1 0.1188 1>
         <rgba 0 1 0.1188 1>
        ]}}
      Random Seed=0
     )
     Passthrough Linear Color=<rgba 1 1 1 1>
    )
    Lifetime Max=1.5
    Lifetime Min=0.3
    Lifetime Mode=::NewEnumerator1
    Sprite Size Mode=::NewEnumerator3
    Uniform Sprite Size=&MaskFloatBySpawnGroup(
     Masked Float=&RandomRangeFloat(Minimum=5, Maximum=12)
     Passthrough Float=&RandomRangeFloat(Minimum=0.6, Maximum=2.5)
    )
    UsePositionOffset=false
    Write Material Random=true
   ) {module=InitializeParticle001}
   &CylinderLocation(Cylinder Height=5, Cylinder Radius=12, MaskSpawnGroup=false, Override Seed=false, UseOwnerScale=true)
   &AddVelocityFromPoint(Module.Use Falloff Distance=false, Origin Offset=<0 0 -50>, Velocity Strength=50)
   &AddVelocity(
    Cone Angle=90
    Cone Axis=<0 0.8 1>
    Module.UseConeFalloff=false
    Module.UseVelDistribution=false
    Override Seed=false
    Velocity Mode=::NewEnumerator2
    Velocity Speed=&RandomRangeFloat(Minimum=325, Maximum=400)
   )
   &SubUVAnimation(
    Evaluation Type=::NewEnumerator0
    "Sprite Renderer "=DataInterface() {propertyValues {SpriteRenderer=&self:Embers_5.NiagaraSpriteRendererProperties_0}}
    SubUV Animation Mode:ENiagara_SubUVLookupModeV2=::NewEnumerator2
    UseEndFrame=false
    UseStartFrame=false
   )
  ]
  particleUpdateScript [
   &ParticleState()
   &GravityForce(Gravity=&RandomRangeVector(Minimum=<0 0 250>, Maximum=<0 0 325>))
   /Niagara/Modules/Update/Forces/CurlNoiseForce(
    Noise Frequency=0.04
    Noise Strength=&FloatFromCurve(FloatCurve=DataInterface() {propertyValues {Curve {keys [<at 0 1>,<at 1 0>]}}}, Scale Curve=500)
    Pan Noise=true
    Pan Noise Field=<0 0 1>
   )
   &Drag(Drag=&RandomRangeFloat(Minimum=5, Maximum=8), Use Linear Drag=true, Use Rotational Drag=false)
   &SolveForcesAndVelocity(Clamp Velocity=false, Limit Acceleration=false)
   &ScaleSpriteSizeBySpeed(
    Max Scale Factor=<0.3 5>
    Sample Scale Factor By Curve=true
    Scale Factor Curve=DataInterface() {propertyValues {Curve {keys [
        <at -0.0077 0.0765 RCIM_Cubic>
        {arriveTangent=1.2463, interpMode=RCIM_Cubic, leaveTangent=1.2463, tangentMode=RCTM_User, time=0.9985, value=0.9549}
       ]}}}
    Velocity Threshold=350
   )
   &ScaleSpriteSize(Uniform Curve Sprite Scale=DataInterface() {propertyValues {Curve {keys [<at 0.8 1 RCIM_Cubic>, {arriveTangent=-6.6126, interpMode=RCIM_Cubic, leaveTangent=-6.6126, tangentMode=RCTM_User, time=1, value=0}]}}})
   &ScaleColor(
    Scale Alpha=&FloatFromCurve(FloatCurve=DataInterface() {propertyValues {Curve {keys [<at 0 1>,<at 1 0>]}}})
    Scale RGB=&MakeVectorFromLinearColorRGB(Color=&ColorFromCurve(ColorCurve=DataInterface() {propertyValues {
        AlphaCurve {keys [<at 0 1 RCIM_Cubic>]}
        BlueCurve {keys [<at 0 15 RCIM_Cubic>,<at 0.2499 5>,<at 0.7114 1 RCIM_Cubic>]}
        GreenCurve {keys [<at 0 15 RCIM_Cubic>,<at 0.2499 5>,<at 0.7114 1 RCIM_Cubic>]}
        RedCurve {keys [<at 0 15 RCIM_Cubic>,<at 0.2499 5>,<at 0.7114 1 RCIM_Cubic>]}
       }}))
    ScaleA=true
    ScaleRGB=true
   )
  ]
  rendererClasses [&NiagaraSpriteRendererProperties]
  renderers [
   {rendererClass=&NiagaraSpriteRendererProperties}
   {rendererClass=&NiagaraSpriteRendererProperties, rendererIndex=1}
  ]
 }
 Fire {
  simTarget=GPUComputeSim
  emitterUpdateScript [
   &EmitterState(Life Cycle Mode=::NewEnumerator1, Loop Behavior=::NewEnumerator1, Loop Duration Mode=::NewEnumerator1, UseLoopDelay=false)
   &SpawnRate(SpawnRate=5, Use Spawn Probability=false)
   &SpawnRate(SpawnGroup=1, SpawnRate=7, Use Spawn Probability=false) {module=SpawnRate001}
   &SpawnRate(SpawnGroup=2, SpawnRate=8, Use Spawn Probability=false) {module=SpawnRate002}
   &SpawnBurst_Instantaneous(Spawn Count=12, Spawn Group=2, Use Spawn Probability=false, UseLoopCountLimit=false)
  ]
  particleSpawnScript [
   /SetVariables(Particles.VisibilityTag=&SpawnGroup()) {isSetParameters=true}
   &InitializeParticle(
    Color=<rgba 8 8 8 1>
    Lifetime Max=1.2
    Lifetime Min=0.9
    Lifetime Mode=::NewEnumerator1
    Position Offset=<0 0 -2>
    Sprite Rotation Mode=::NewEnumerator2
    Sprite Size Mode=::NewEnumerator3
    Uniform Sprite Size=&MaskFloatBySpawnGroup(
     Masked Float=&RandomRangeFloat(Minimum=65, Maximum=75)
     Passthrough Float=&MaskFloatBySpawnGroup(
      Masked Float=&RandomRangeFloat(Minimum=50, Maximum=60)
      Passthrough Float=&RandomRangeFloat(Minimum=35, Maximum=40)
      Spawn Group Mask=1
     )
    )
    UsePositionOffset=true
    Write Material Random=true
   ) {module=InitializeParticle001}
   &Color(Color=&MaskLinearColorBySpawnGroup(
     Masked Linear Color=&Multiply_LinearColorByFloat(
      Linear Color=<rgba 4 0.3924 0 0.8>
      Float=&FNTimeOfDayScalar(Night Scale=1, Morning Scale=1, Day Scale=0.8, Evening Scale=1)
     )
     Passthrough Linear Color=&MaskLinearColorBySpawnGroup(
      Masked Linear Color=&Multiply_LinearColorByFloat(
       Linear Color=&SelectLinearColorFromArray(
        Array Sampling Mode=::NewEnumerator0
        Direct Array Index=?
        Color Selection Array=DataInterface() {propertyValues {ColorData [
           <rgba 0 0.5951 3 1>
           <rgba 0 0.5951 3 1>
           <rgba 0 0.3967 2 1>
           <rgba 0.5102 0.1696 5 1>
           <rgba 0.2041 0.0679 2 1>
           <rgba 0.2041 0.0679 2 1>
           <rgba 0 2 0.2377 1>
          ]}}
        Random Seed=0
       )
       Float=&FNTimeOfDayScalar(Night Scale=1, Morning Scale=1, Evening Scale=1)
      )
      Passthrough Linear Color=&SelectLinearColorFromArray(
       Array Sampling Mode=::NewEnumerator0
       Direct Array Index=?
       Color Selection Array=DataInterface() {propertyValues {ColorData [<rgba 4 0.3924 0 0.55>, <rgba 0 0.3967 2 1>, <rgba 0.2041 0.0679 2 1>, <rgba 0 2 0.2377 1>]}}
       Random Seed=0
      )
      Spawn Group Mask=1
     )
    ))
   &CylinderLocation(
    Cylinder Height=1
    Cylinder Radius=&MaskFloatBySpawnGroup(
     Masked Float=0
     Passthrough Float=&MaskFloatBySpawnGroup(Masked Float=0, Passthrough Float=12, Spawn Group Mask=1)
    )
    MaskSpawnGroup=false
    Non Uniform Scale:Vector3f=<1 1 1.5>
    Override Seed=false
    UseOwnerScale=true
   ) {enabled=false}
   &ShapeLocation(
    MaskSpawnGroup=false
    Override Seed=false
    Ring Radius=12
    Rotation Axis=<0 0 1>
    Rotation Mode=::NewEnumerator1
    Shape Primitive=::NewEnumerator4
    UseOwnerScale=true
   )
   &AddVelocity(
    Cone Angle=25
    Cone Axis=<0 0 1>
    Module.UseConeFalloff=false
    Module.UseVelDistribution=false
    Override Seed=false
    Velocity Mode=::NewEnumerator2
    Velocity Speed=&MaskFloatBySpawnGroup(
     Masked Float=&RandomRangeFloat(Minimum=25, Maximum=50)
     Passthrough Float=&MaskFloatBySpawnGroup(Masked Float=&RandomRangeFloat(Minimum=25, Maximum=50), Passthrough Float=0, Spawn Group Mask=1)
    )
   )
   &DynamicMaterialParameters(
    Index 0 Param 2=&RandomRangeFloat(Minimum=0, Maximum=4)
    Index 0 Param 4=4
    Module.Param0WriteEnabled=false
    Module.Param1WriteEnabled=true
    Module.Param2WriteEnabled=false
    Module.Param3WriteEnabled=true
   ) {module=DynamicMaterialParameters001}
  ]
  particleUpdateScript [
   &ParticleState()
   &SubUVAnimation(
    "Sprite Renderer "=DataInterface() {propertyValues {SpriteRenderer=&self:Fire_3.NiagaraSpriteRendererProperties_0}}
    UseEndFrame=false
    UseStartFrame=false
   )
   &SubUVAnimation(
    "Sprite Renderer "=DataInterface() {propertyValues {SpriteRenderer=&self:Fire_3.NiagaraSpriteRendererProperties_1}}
    UseEndFrame=false
    UseStartFrame=false
   ) {module=SubUVAnimation001}
   &SubUVAnimation(
    "Sprite Renderer "=DataInterface() {propertyValues {SpriteRenderer=&self:Fire_3.NiagaraSpriteRendererProperties_2}}
    UseEndFrame=false
    UseStartFrame=false
   ) {module=SubUVAnimation002}
   &Drag(Drag=&RandomRangeFloat(Minimum=0.8, Maximum=1.2), Use Linear Drag=true, Use Rotational Drag=false)
   &ScaleColor(
    Scale Alpha=&Multiply_Float(
     A=1
     B=&FloatFromCurve(FloatCurve=DataInterface() {propertyValues {Curve {keys [
          {arriveTangent=2.4553, interpMode=RCIM_Cubic, leaveTangent=2.4553, time=0, value=0}
          {arriveTangent=-0.0116, interpMode=RCIM_Cubic, leaveTangent=-0.0116, tangentMode=RCTM_User, time=0.4073, value=1}
          {arriveTangent=-2.0743, interpMode=RCIM_Cubic, leaveTangent=-2.0743, tangentMode=RCTM_User, time=0.7, value=0.2}
          {arriveTangent=-0.6667, interpMode=RCIM_Cubic, leaveTangent=-0.6667, time=1, value=0}
         ]}}})
    )
    Scale RGB=&VectorFromCurve(VectorCurve=DataInterface() {propertyValues {
       XCurve {keys [<at 0.3242 2 RCIM_Cubic>,<at 0.9878 0.9731>]}
       YCurve {keys [<at 0.3242 2 RCIM_Cubic>,<at 1 0.9731>]}
       ZCurve {keys [<at 0.3242 2 RCIM_Cubic>,<at 1 0.9731>]}
      }})
    ScaleA=true
    ScaleRGB=false
   ) {module=ScaleColor001}
   &ScaleSpriteSize(Scale Factor=&Vector2DFromCurve(
     Vector2Curve=DataInterface() {propertyValues {
       XCurve {keys [<at 0 0.3 RCIM_Cubic>, {arriveTangent=0.3, interpMode=RCIM_Cubic, leaveTangent=0.3, time=0.3, value=1}, <at 1 0.6>]}
       YCurve {keys [<at 0 0.3 RCIM_Cubic>, {arriveTangent=0.3, interpMode=RCIM_Cubic, leaveTangent=0.3, time=0.3, value=1}, <at 1 0.6>]}
      }}
     ScaleCurve=<1 1>
    ))
   &DynamicMaterialParameters(
    Index 0 Param 1=&FloatFromCurve(FloatCurve=DataInterface() {propertyValues {Curve {keys [<at 0 1>]}}})
    Index 0 Param 2=0.5
    Index 0 Param 3=&FloatFromCurve(FloatCurve=DataInterface() {propertyValues {Curve {keys [<at 0 0.5 RCIM_Cubic>, {arriveTangent=1.5, interpMode=RCIM_Cubic, leaveTangent=1.5, time=0.5235, value=0.5}, <at 1 2>]}}})
    Module.Param0WriteEnabled=true
    Module.Param1WriteEnabled=false
    Module.Param2WriteEnabled=true
    Module.Param3WriteEnabled=false
   )
   &SolveForcesAndVelocity(Clamp Velocity=false, Limit Acceleration=false)
   /SetVariables(Particles.CoreDP:Vector4f=&MakeVector4(
     X=&FloatFromCurve(FloatCurve=DataInterface() {propertyValues {Curve {keys [<at 0 1>]}}})
     Y=&RandomRangeFloat(Minimum=0, Maximum=4)
     Z=&FloatFromCurve(FloatCurve=DataInterface() {propertyValues {Curve {keys [<at 0 0.5 RCIM_Cubic>, {arriveTangent=1.5, interpMode=RCIM_Cubic, leaveTangent=1.5, time=0.5235, value=0.5}, <at 1 2>]}}})
     W=2.5
    )) {isSetParameters=true}
  ]
  rendererClasses [&NiagaraSpriteRendererProperties]
  renderers [
   {rendererClass=&NiagaraSpriteRendererProperties}
   {rendererClass=&NiagaraSpriteRendererProperties, rendererIndex=1}
   {rendererClass=&NiagaraSpriteRendererProperties, rendererIndex=2}
  ]
 }
 Fire_Spline {
  emitterUpdateScript [
   &EmitterState(Life Cycle Mode=::NewEnumerator1, Loop Behavior=::NewEnumerator1, Loop Duration Mode=::NewEnumerator1, UseLoopDelay=false)
   &SpawnBurst_Instantaneous(Use Spawn Probability=false, UseLoopCountLimit=false)
   &SpawnBurst_Instantaneous(Spawn Group=1, Use Spawn Probability=false, UseLoopCountLimit=false) {module=SpawnBurst_Instantaneous001}
   &SpawnBurst_Instantaneous(Spawn Group=2, Use Spawn Probability=false, UseLoopCountLimit=false) {module=SpawnBurst_Instantaneous002}
  ]
  particleSpawnScript [
   /SetVariables(Particles.VisibilityTag=&SpawnGroup()) {isSetParameters=true}
   &InitializeParticle(
    Lifetime=5
    Mesh Scale=&MaskVectorBySpawnGroup(
     Masked Vector=<1 1 2>
     Passthrough Vector=&MaskVectorBySpawnGroup(Masked Vector=<1 1 1.4>, Passthrough Vector=<1 1 1.6>, Spawn Group Mask=1)
    )
    Mesh Scale Mode=::NewEnumerator4
    Position Offset=&MaskVectorBySpawnGroup(
     Masked Vector=<0 0 -3>
     Passthrough Vector=&MaskVectorBySpawnGroup(Masked Vector=<-5 0 -5>, Passthrough Vector=<3 0 -5>, Spawn Group Mask=1)
    )
    UsePositionOffset=true
    Write Material Random=true
   )
   &InitialMeshOrientation(
    Mesh Axis To Orient=<0 0 1>
    Mesh Orientation Mode=::NewEnumerator2
    Orientation Vector=<0 0 1>
    Rotation=<0.02 0 0>
    Use Rotation Vector=true
   )
   &InitialMeshOrientation(
    Mesh Orientation Mode=::NewEnumerator2
    Rotation=&MaskVectorBySpawnGroup(
     Masked Vector=<-0.03 0 0>
     Passthrough Vector=&MaskVectorBySpawnGroup(Masked Vector=<-0.05 0 0>, Passthrough Vector=<0.03 0 0>, Spawn Group Mask=1)
    )
    Use Rotation Vector=true
   ) {module=InitialMeshOrientation001}
   &CameraOffset(Camera Offset Amount=&MaskFloatBySpawnGroup(
     Masked Float=0
     Passthrough Float=&MaskFloatBySpawnGroup(Masked Float=0, Passthrough Float=0, Spawn Group Mask=1)
    ))
  ]
  particleUpdateScript [
   &ParticleState(Kill Particles When Lifetime Has Elapsed=false, Let Infinitely Lived Particles Die When Emitter Deactivates=true)
   &InterpolateOverTime(Rate Of Change=8, Target Vector Value:NiagaraPosition=$Particles.Position, Variable Type=::NewEnumerator2)
   /SetVariables(Particles.Interpolate Dynamic=&MakeFloat4FromVectorAndFloat("Vector (RGB)"=$Particles.InterpolateOverTime.InterpolateOverTime_V3_.Moving Average, "Float (Alpha)"=0.7)) {isSetParameters=true}
   &DynamicMaterialParameters(
    Index 0 Param 3=3.089
    Module.Index0Float0Write=true
    Module.Index0Float1Write=true
    Module.Index0Float2Write=true
    Module.Index0Float3Write=true
   )
   /SetVariables(Particles.MainColor:LinearColor=&MultiplyLinearColor(
     Linear Color=&MaskLinearColorBySpawnGroup(
      Masked Linear Color=&LerpLinearColors(
       StartColor=<rgba 1 0.2295 0 1>
       EndColor=<rgba 1 0.4884 0 1>
       LerpFactor=&Waveform("Waveform [1]"=::NewEnumerator0, "[1] Frequency"=0.3, "[1] Amplitude Min/Max"=<0 1>)
      )
      Passthrough Linear Color=&MaskLinearColorBySpawnGroup(
       Masked Linear Color=&LerpLinearColors(
        StartColor=<rgba 0.046 1 0 1>
        EndColor=<rgba 0.0006 0.84 1 1>
        LerpFactor=&Waveform("Waveform [1]"=::NewEnumerator0, "[1] Frequency"=0.45, "[1] Amplitude Min/Max"=<0 1>)
       )
       Passthrough Linear Color=&LerpLinearColors(
        StartColor=<rgba 1 0 0.5383 1>
        EndColor=<rgba 0.1408 0.0013 1 1>
        LerpFactor=&Waveform("Waveform [1]"=::NewEnumerator0, "[1] Frequency"=0.45, "[1] Amplitude Min/Max"=<0 1>)
       )
       Spawn Group Mask=1
      )
     )
     Scale Factor=/Niagara/DynamicInputs/Multiply/Multiply_Vector4_ByFloat(
      Vector4:Vector4f=&MakeVector4_FromLinearColor(LinearColor=&ColorFromCurve(ColorCurve=DataInterface() {propertyValues {
          AlphaCurve {keys [<at 0 0 RCIM_Cubic>,<at 0.1 1>]}
          BlueCurve {keys [<at 0 1 RCIM_Cubic>,<at 1 1 RCIM_Cubic>]}
          GreenCurve {keys [<at 0 1 RCIM_Cubic>,<at 1 1 RCIM_Cubic>]}
          RedCurve {keys [<at 0 1 RCIM_Cubic>,<at 1 1 RCIM_Cubic>]}
         }}))
      Float:NiagaraFloat=&FNTimeOfDayScalar(Night Scale=1, Morning Scale=1, Evening Scale=1)
     )
    )) {isSetParameters=true}
  ]
  rendererClasses [&NiagaraMeshRendererProperties]
  renderers [
   {rendererClass=&NiagaraMeshRendererProperties, rendererIndex=0}
   {rendererClass=&NiagaraMeshRendererProperties, rendererIndex=1}
   {rendererClass=&NiagaraMeshRendererProperties, rendererIndex=2}
  ]
 }
 MusicTrail {
  emitterUpdateScript [
   &EmitterState(Life Cycle Mode=::NewEnumerator1, Loop Behavior=::NewEnumerator1, Loop Delay=0.24, Loop Duration Mode=::NewEnumerator1, UseLoopDelay=false)
   &SpawnRate(
    SpawnRate=&FloatFromCurve(FloatCurve=DataInterface() {propertyValues {Curve {keys [<at 0.24 0>,<at 0.36 1>]}}}, CurveIndex=$Emitter.Age, Scale Curve=60)
    Use Spawn Probability=false
   )
  ]
  particleSpawnScript [
   &InitializeParticle(
    Color=<rgba 10 10 10 0.6>
    Position Offset=<0 0 -8>
    Ribbon Width=20
    Ribbon Width Mode=::NewEnumerator1
    Sprite Size=<70 25>
    Sprite Size Mode=::NewEnumerator4
    UsePositionOffset=true
    Write Material Random=true
   )
   &AddVelocity(Velocity=<0 0 125>)
  ]
  particleUpdateScript [
   &ParticleState()
   &WindForce(Wind Speed=<-45 0 0>)
   &Drag(Use Linear Drag=true, Use Rotational Drag=true)
   &SolveForcesAndVelocity(Clamp Velocity=false, Limit Acceleration=false)
   &ScaleColor(
    Linear Color Curve=DataInterface() {propertyValues {
      AlphaCurve {keys [<at 0.225 0>,<at 0.5 1>,<at 1 1>]}
      BlueCurve {keys [<at 0.3798 0 RCIM_Cubic>,<at 0.4981 0>,<at 0.6274 0.7585>,<at 0.75 1>]}
      GreenCurve {keys [<at 0.3798 0.2667 RCIM_Cubic>,<at 0.4981 1>,<at 0.6274 0.345>,<at 0.75 0>]}
      RedCurve {keys [<at 0.3798 1 RCIM_Cubic>,<at 0.4981 0.5265>,<at 0.6274 0>,<at 0.75 0.4665>]}
     }}
    Scale Mode=::NewEnumerator2
   )
   &ScaleColor(
    Linear Color Curve=DataInterface() {propertyValues {
      AlphaCurve {keys [<at 0 0>,<at 0.2338 1>]}
      BlueCurve {keys [<at 0.2948 0 RCIM_Cubic>]}
      GreenCurve {keys [<at 0.2948 0.2667 RCIM_Cubic>]}
      RedCurve {keys [<at 0.2948 1 RCIM_Cubic>]}
     }}
    Scale Mode=::NewEnumerator2
   ) {enabled=false, module=ScaleColor001}
   &ScaleRibbonWidth(Ribbon Width Scale=&FloatFromCurve(FloatCurve=DataInterface() {propertyValues {Curve {keys [
         {arriveTangent=9.378, interpMode=RCIM_Cubic, leaveTangent=9.378, tangentMode=RCTM_User, time=0, value=0.3}
         {arriveTangent=-1.4098, interpMode=RCIM_Cubic, leaveTangent=-1.4098, tangentMode=RCTM_User, time=0.2, value=1}
         <at 1 0.4 RCIM_Cubic>
        ]}}}))
   /SetVariables(Particles.Wave Pos:NiagaraPosition=&Add_Vector(
     A=Linked() {linkedVariable {name=Particles.Position, type=NiagaraPosition}}
     B=&MakeVector(
      X=&Sine(
       Normalized Angle=&Add_Float(A=$Emitter.Age, B=&Multiply_Float(A=$Particles.NormalizedAge, B=-0.5))
       Period=1.8
       Scale=&FloatFromCurve(
        FloatCurve=DataInterface() {propertyValues {Curve {keys [<at 0 0>, {arriveTangent=1, interpMode=RCIM_Cubic, leaveTangent=1, time=0.2304, value=0.7182}, <at 1 1>]}}}
        Scale Curve=3.75
       )
       Bias=0
      )
      Y=0
      Z=0
     )
    )) {isSetParameters=true}
  ]
  rendererClasses [&NiagaraRibbonRendererProperties]
  renderers [{rendererClass=&NiagaraRibbonRendererProperties}]
 }
 Scorch {
  emitterUpdateScript [
   &EmitterState(
    Enable Distance Culling=true
    Enable Visibility Culling=true
    Inactive Response=::NewEnumerator1
    Life Cycle Mode=::NewEnumerator1
    Loop Behavior=::NewEnumerator1
    Loop Duration Mode=::NewEnumerator1
    Module.bUseMaxDistance=true
    Module.bUseMinDistance=false
    Scalability Mode=::NewEnumerator1
    UseLoopDelay=false
   )
   &SpawnBurst_Instantaneous(Spawn Time=0.12, Use Spawn Probability=false, UseLoopCountLimit=false)
  ]
  particleSpawnScript [
   &InitializeParticle(
    Lifetime=5
    Mesh Scale=<1 0.9 1>
    Mesh Scale Mode=::NewEnumerator4
    Sprite Size Mode=::NewEnumerator3
    Uniform Sprite Size=85
    UsePositionOffset=false
    Write Material Random=true
   )
   &DynamicMaterialParameters(
    Index 0 Param 1=5
    Index 0 Param 2=1
    Module.Index0Float0Write=true
    Module.Index0Float1Write=true
    Module.Index0Float2Write=true
    Module.Index0Float3Write=true
   )
   &InitialMeshOrientation(
    Mesh Axis To Orient=<0 1 0>
    Mesh Orientation Mode=::NewEnumerator2
    Orientation Vector=<0 0 1>
    Rotation=&MaskVectorBySpawnGroup(
     Masked Vector=<0 0.25 0>
     Passthrough Vector=&MaskVectorBySpawnGroup(Masked Vector=<0 -0.15 0>, Passthrough Vector=<0 0.05 0>, Spawn Group Mask=1)
    )
    Use Rotation Vector=true
   )
   &CameraOffset(Camera Offset Amount=20)
   /SetVariables() {isSetParameters=true}
  ]
  particleUpdateScript [
   &ParticleState(
    Kill Particles When Lifetime Has Elapsed=false
    Let Infinitely Lived Particles Die When Emitter Deactivates=true
    Loop Particles Lifetime=true
   )
   /SetVariables(Particles.Jitter:NiagaraFloat=&Waveform("Waveform [1]"=::NewEnumerator12, "[1] Frequency"=12, "[1] Amplitude Min/Max"=<0.85 1.1>)) {isSetParameters=true}
   &FN_Sprite_Attributes(
    Alignment:Vector3f=<1 0 0>
    Color=&MultiplyLinearColor(
     Linear Color=&ColorFromCurve(ColorCurve=DataInterface() {propertyValues {
        AlphaCurve {keys [<at 0 1>]}
        BlueCurve {keys [<at 0 0>, {interpMode=RCIM_Cubic, tangentMode=RCTM_Break, time=1, value=0}]}
        GreenCurve {keys [<at 0 0.1667>, {interpMode=RCIM_Cubic, leaveTangent=-0.3515, time=1, value=0.1097}]}
        RedCurve {keys [<at 0 1>, {arriveTangent=0.1102, interpMode=RCIM_Cubic, leaveTangent=0.1102, tangentMode=RCTM_User, time=1, value=0.4115}]}
       }})
     Scale Factor=&MakeFloat4FromVectorAndFloat(
      "Vector (RGB)"=&VectorFromFloat(Value=0.6)
      "Float (Alpha)"=&Multiply_Float(A=0.6, B=&Waveform("Waveform [1]"=::NewEnumerator2, "[1] Frequency"=1, "[1] Amplitude Min/Max"=<0.6 1.5>))
     )
    )
    Create Alignment Attribute=true
    Create Camera Offset Attribute=true
    Create Color Attribute=true
    Create Facing Attribute=true
    Create Position Attribute=true
    Create SpriteSize Attribute=true
    Facing:Vector3f=<0 0 1>
    SpriteSize=&Vector2DFromFloat(Value=95)
    Write to Module Namespace=true
   ) {module=FN_Sprite_Attributes001}
   &ScaleColor(
    Linear Color Curve=DataInterface() {propertyValues {
      AlphaCurve {keys [<at 0 1>]}
      BlueCurve {keys [<at 0 0>,<at 0.1412 1>,<at 0.2673 0>,<at 0.4 1>,<at 0.5496 0.0048>,<at 0.6704 0>,<at 0.8 0>,<at 0.8906 0>,<at 1 0>]}
      GreenCurve {keys [<at 0 0.3325>,<at 0.1412 0.2535>,<at 0.2673 0.3325>,<at 0.4 0>,<at 0.5496 0.7328>,<at 0.6704 0.3325>,<at 0.8 1>,<at 0.8906 0.3325>,<at 1 0.0576>]}
      RedCurve {keys [<at 0 1>,<at 0.1412 0>,<at 0.2673 1>,<at 0.4 0.7543>,<at 0.5496 1>,<at 0.6704 1>,<at 0.8 0.4661>,<at 0.8906 1>,<at 1 1>]}
     }}
    Scale Mode=::NewEnumerator2
   )
   &FN_Light_Attributes(
    Color=$Particles.Color
    Create Color Attribute=true
    Create Exponent Attribute=true
    Create Position Attribute=true
    Create Radius Attribute=true
    Exponent=4
    Position=<0 0 5>
    Radius=&Multiply_Float(A=$Particles.Jitter, B=100)
    Scale Color=&FN_FrontEndFloatModifier(Gameplay=2.5, Front End=0)
   )
   /SetVariables(Particles.ScorchDP:Vector4f=<0 0 0 0>) {isSetParameters=true}
   &ScaleMeshSize(Scale Factor=&VectorFromFloat(Value=&FloatFromCurve(FloatCurve=DataInterface() {propertyValues {Curve {keys [<at 0 0>,<at 0.25 1>]}}}, CurveIndex=$Emitter.Age)))
   /SetVariables(
    Particles.FireDP:Vector4f=<0 0 0 0>
    Particles.FirePos:NiagaraPosition=&AddVectorToPosition(Position=$Engine.Emitter.SimulationPosition, Vector=<0 0 -20>)
    Particles.MainColor:LinearColor=&MultiplyLinearColor(
     Linear Color=&MultiplyLinearColor(
      Linear Color=<rgba 2.5 0.2805 0.001 1>
      Scale Factor=&MakeVector4_FromLinearColor(LinearColor=&ColorFromCurve(
        ColorCurve=DataInterface() {propertyValues {
          AlphaCurve {keys [<at 0 0 RCIM_Cubic>,<at 0.1 1>]}
          BlueCurve {keys [<at 0 1 RCIM_Cubic>,<at 1 1 RCIM_Cubic>]}
          GreenCurve {keys [<at 0 1 RCIM_Cubic>,<at 1 1 RCIM_Cubic>]}
          RedCurve {keys [<at 0 1 RCIM_Cubic>,<at 1 1 RCIM_Cubic>]}
         }}
        CurveIndex=$Emitter.Age
       ))
     )
     Scale Factor=&MakeVector4_FromFloat(Float=&FNTimeOfDayScalar(Night Scale=0.75, Morning Scale=1, Day Scale=0.7, Evening Scale=1))
    )
   ) {isSetParameters=true}
  ]
  rendererClasses [
   &NiagaraLightRendererProperties
   &NiagaraSpriteRendererProperties
   &NiagaraMeshRendererProperties
  ]
  renderers [
   {rendererClass=&NiagaraLightRendererProperties, rendererIndex=0}
   {rendererClass=&NiagaraSpriteRendererProperties, rendererIndex=1}
   {rendererClass=&NiagaraMeshRendererProperties, rendererIndex=2}
  ]
 }
 SmokeLW {rendererClasses [], renderers []}
}
summary {
 systemName=NS_Drum_DrumstickCampfire_Emote
 emitters [
  {emitterName=Embers, rendererClasses [&NiagaraSpriteRendererProperties]}
  {emitterName=Fire, simTarget=GPUComputeSim, rendererClasses [&NiagaraSpriteRendererProperties]}
  {
   emitterName=Scorch
   rendererClasses [
    &NiagaraLightRendererProperties
    &NiagaraSpriteRendererProperties
    &NiagaraMeshRendererProperties
   ]
  }
  {emitterName=MusicTrail, rendererClasses [&NiagaraRibbonRendererProperties]}
  {emitterName=SmokeLW, rendererClasses []}
  {emitterName=Fire_Spline, rendererClasses [&NiagaraMeshRendererProperties]}
  {bEnabled=false, emitterName=Center_flame, rendererClasses [&NiagaraMeshRendererProperties]}
 ]
 userVariables [
  User.bIsFrontEnd {type=NiagaraBool}
  User.bIsFrontEndPreview {type=NiagaraBool}
 ]
}
